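# Post-process style AnimBlueprint for the GuardHence male body cosmetic that follows the parent character's pose and adds secondary motion to its bow bones.
# Copy Pose From Mesh feeds a Control Rig, an additive base_bow offset and quarter-strength AnimDynamics on dyn_bow.
# Each update stores gravity and RigidBody settings from the Head's parameter asset, but no node shown reads them.
# Two Constraint nodes and a second AnimDynamics node sit unconnected.
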AnimBlueprint /BRCosmetics/Characters/Player/Male/Medium/Bodies/M_MED_GuardHence/Meshes/M_MED_GuardHence_AnimBP
target_skeleton=&/M_MED_GuardHence_Skeleton
graphs [
 AnimGraph {nodes [
   Anim Dynamics - Bone: dyn_bow(Alpha=1, ExternalForce="0,0,0") {
    angularDampingOverride=1
    angularSpringConstant=50
    bOverrideAngularDamping=true
    bUseGravityOverride=true
    gravityScale=0.35
    simulationSpace=RootRelative
    worldSpaceMinimumScale=0
    boundBone {boneName=dyn_bow}
    componentAppliedLinearAccClamp=<0 0 0>
    gravityOverride {z=-980}
    physicsBodyDefinitions [{
      boundBone {boneName=dyn_bow}
      boxExtents=<10 10 10>
      constraintSetup {angularLimitsMax=<25 0 0>, angularLimitsMin=<-25 0 0>, angularTarget=<1 0 0>}
      localJointOffset=<4 0 0>
     }]
    relativeSpaceBone {boneName=spine_05}
    simSpaceSettings {maxAngularAcceleration=32, maxAngularVelocity=8, simSpaceAngularAlpha=1}
   }
   "Constraint - base_bow (PPPPPPPPP)"!(
    ComponentPose
    Alpha=1
    ConstraintWeights_0=0.4548
    ConstraintWeights_1:float=0.3145
    ConstraintWeights_2:float=0.1481
    ConstraintWeights_3:float=0.0506
    ConstraintWeights_4:float=0.0086
    ConstraintWeights_5:float=0.0086
    ConstraintWeights_6:float=0.0074
    ConstraintWeights_7:float=0.0038
    ConstraintWeights_8:float=0.0036
   ) -> (Pose) {
    boneToModify {boneName=base_bow}
    constraintSetup [
     {targetBone {boneName=spine_04}}
     {targetBone {boneName=spine_05}}
     {targetBone {boneName=spine_03}}
     {targetBone {boneName=spine_02}}
     {targetBone {boneName=clavicle_r}}
     {targetBone {boneName=clavicle_l}}
     {targetBone {boneName=deform_pec_l}}
     {targetBone {boneName=deform_pec_base_l}}
     {targetBone {boneName=deform_pec_r}}
    ]
    constraintWeights [1,1,1,1,1,1,1,1,1]
   }
   "Constraint - base_bow (P)"!(ComponentPose, Alpha=1, ConstraintWeights_0=1, ConstraintWeights_1:float=0) -> (Pose) {boneToModify {boneName=base_bow}, constraintSetup [{targetBone {boneName=dyn_bow}}], constraintWeights [1,0]}
   n1 = Control Rig(Source=Copy Pose From Mesh() {bUseAttachedParent=true}) {controlRigAssetReference {blueprintRigClass=&/M_MED_GuardHence_CtrlRig.M_MED_GuardHence_CtrlRig_C}}
   n2 = Anim Dynamics - Bone: dyn_bow(
    ComponentPose="Transform (Modify) Bone - Bone: base_bow"(ComponentPose=Local To Component(LocalPose=@n1), Alpha=1, Translation="0.000,0.626,0.660", Rotation="0,0,0", Scale="1,1,1") {translationMode=BMM_Additive, boneToModify {boneName=base_bow}, translation {y=0.626, z=0.66}}
    Alpha=0.25
    ExternalForce="0,0,0"
   ) {
    angularDampingOverride=1
    bOverrideAngularDamping=true
    bUseGravityOverride=true
    gravityScale=0.35
    simulationSpace=RootRelative
    worldSpaceMinimumScale=0
    boundBone {boneName=dyn_bow}
    componentAppliedLinearAccClamp=<0 0 0>
    componentLinearVelScale {z=1}
    gravityOverride {z=-980}
    physicsBodyDefinitions [{
      boundBone {boneName=dyn_bow}
      boxExtents=<10 10 10>
      constraintSetup {twistAxis=AxisY, angularLimitsMax=<10 0 0>, angularLimitsMin=<-10 0 0>, angularTarget=<0 1 0>}
      localJointOffset=<0 5 0>
     }]
    relativeSpaceBone {boneName=spine_05}
    simSpaceSettings {maxAngularAcceleration=36, maxAngularVelocity=12, simSpaceAngularAlpha=1}
   }
   Output Pose(Result=Component To Local(ComponentPose=@n2)) {layerGroup=None}
  ]}
 EventGraph {nodes [
   Event BlueprintUpdateAnimation()
   > n1 = "Gravity Override State Function (Parameter Asset) (Nativized)"(
    CharacterPartAnimInstance=Self-Reference()
    GameTimeInSeconds=GetGameTimeInSeconds()
    Parameters=/BRCosmetics/Characters/Player/Male/Medium/Heads/M_MED_GuardHence_Head/Meshes/CustomSkeleton/M_MED_GuardHence_Parameters
   )
   > Set GravityOverride(GravityOverride=@n1)
   > Set RBAN Settings Out(RBAN Settings Out=@n1.OutRigidBodyAnimNodeParameters)
  ]}
]
variables [
 GravityOverride {type=Vector}
 RBAN Settings Out {type=FortRigidBodyAnimNodeParameters}
]
